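# Additive, unlit, two-sided impact effect material for the Bizarre pickaxe, usable on Niagara sprites and mesh particles.
# UseParticles blends two looks, with emissive passed through MF_FX_EmissiveAdjustments.
# One multiplies circuitboard, flame and smoke-distorted trail textures, scrolled by dynamic parameter, with gradient tint and negative-desaturation saturation boost.
# The other draws rounded squares with noisy glowing edges, made by subtracting a dynamically scaled copy of the shape.
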Material /BRCosmetics/Weapons/FORT_Melee/Pickaxe_Bizarre/FX/Material/Impact/M_Pickaxe_Bizarre_Impact
graph {
 properties {
  bUsedWithNiagaraMeshParticles=true
  bUsedWithNiagaraSprites=true
  blendMode=BLEND_Additive
  pixelDepthOffsetMode=PDOM_Legacy
  shadingModel=MSM_Unlit
  twoSided=true
  usedShadingModels=Unlit
 }
 nodes [
  n1 = TextureSample(UVs=TextureCoordinate() {uTiling=2}) {texture=/Game/Effects/Fort_Effects/Textures/T_FX_Circuitboard}
  n2 = TextureCoordinate() {uTiling=1.4}
  n3 = Add(A=TextureSample(UVs=Panner(Coordinate=@n2) {speedY=-0.3}) {samplerType=SAMPLERTYPE_LinearGrayscale, texture=/BRCosmetics/Weapons/FORT_Melee/Pickaxe_Cube_Ninja_Male/FX/T_FluidFlameTile}.R) {constB=0.35}
  n4 = TextureCoordinate() {uTiling=2}
  n5 = Panner(Coordinate=TextureCoordinate()) {speedY=0.3}
  n6 = Multiply(A=ComponentMask(None=TextureSample(UVs=@n5) {texture=&T_Inky_Smoke_Tile}.RGB) {g=true, r=true}) {constB=1.4}
  n7 = TextureSample(UVs=Panner(Coordinate=Add(A=@n4, B=@n6), Time=DynamicParameter().Param1) {speedY=1}) {
   samplerSource=SSM_Clamp_WorldGroupSettings
   samplerType=SAMPLERTYPE_LinearColor
   texture=/BRCosmetics/Athena/Effects/Particles/Skydiving/NS_Contrail_CubeNinja/T_CubeNinja_Trail
  }
  n8 = ComponentMask(None=TextureCoordinate()) {g=true}
  n9 = Multiply(A=Multiply(A=@n7.R, B=Multiply(A=@n1.R, B=@n3)), B=@n8)
  n10 = Constant() {r=20}
  n11 = VectorParameter() {parameterName=ColorA, defaultValue=<rgba 0.5866 0 1 1>}
  n12 = VectorParameter() {parameterName=ColorB, defaultValue=<rgba 0.5866 0 1 1>}
  n13 = Multiply(A=@n8, B=@n8)
  n14 = ParticleColor()
  n15 = Multiply(A=Multiply(A=Multiply(A=@n9, B=@n10), B=LinearInterpolate(A=@n11.RGB, B=@n12.RGB, Alpha=@n13)), B=@n14.RGB)
  n16 = Constant() {r=-3}
  n17 = ParticleColor()
  n18 = TextureSample() {samplerType=SAMPLERTYPE_Grayscale, texture=/Game/UI/Frontend/ItemShop/Com/T_vote_feedback_rounded_square}
  n19 = TextureCoordinate() {uTiling=0.2}
  n20 = Add(A=TextureSample(UVs=Panner(Coordinate=@n19) {speedX=0.5}) {samplerType=SAMPLERTYPE_Grayscale, texture=/BRCosmetics/Weapons/FORT_Melee/Pickaxe_Cube_Ninja_Male/FX/T_CellNoise}.R) {constB=0.2}
  n21 = DynamicParameter()
  n22 = Subtract(A=Constant(), B=@n21.Param1)
  n23 = Add(A=Constant() {r=1}, B=@n21.Param1)
  n24 = MaterialFunctionCall(Float2=TextureCoordinate()) {materialFunction=&BreakOutFloat2Components}
  n25 = TextureSample(UVs=AppendVector(
    A=LinearInterpolate(A=@n22, B=@n23, Alpha=@n24.R) {constA=-0.1, constB=1.1}
    B=LinearInterpolate(A=@n22, B=@n23, Alpha=@n24.G) {constA=-0.1, constB=1.1}
   )) {
   samplerSource=SSM_Clamp_WorldGroupSettings
   samplerType=SAMPLERTYPE_Grayscale
   texture=&T_vote_feedback_rounded_square
  }
  n26 = Subtract(A=@n18.R, B=@n25.R)
  n27 = Add(
   A=Multiply(A=@n18.R, B=ScalarParameter() {parameterName=Particle_Multply})
   B=Multiply(A=Multiply(A=@n20, B=@n26), B=ScalarParameter() {defaultValue=30, parameterName=Particle_edge})
  )
  MaterialFunctionCall(Emissive=LinearInterpolate(A=Desaturation(None=@n15, Fraction=@n16), B=Multiply(A=@n17.RGB, B=@n27), Alpha=ScalarParameter() {parameterName=UseParticles})) {materialFunction=&MF_FX_EmissiveAdjustments}
  n28 = Add(A=@n7.R, B=@n1.R)
  Reroute(None=LinearInterpolate(
    A=Multiply(A=@n14.A, B=@n28)
    B=TextureSample() {samplerType=SAMPLERTYPE_Grayscale, texture=&T_vote_feedback_rounded_square}.R
    Alpha=ScalarParameter() {parameterName=UseParticles}
   ))
 ]
}
